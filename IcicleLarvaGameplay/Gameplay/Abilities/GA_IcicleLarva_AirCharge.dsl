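; Icicle Larva air charge ability that smashes through buildings and hands off to a punch ability on hitting something it cannot destroy.
; A looping timer runs a targeting preset, killing damageable BuildingActors while excluding gameplay buildings and Spires.
; Other hits set a flag that activates the punch and ends the ability, which clears the timer.
; A flag with a timed reset throttles the destruction GameplayCue.
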(event Ability|EventActivateAbility
  (|CallParentFunction)
  (bind _returnvalue (Utilities|Time|SetTimerbyEvent (delegate Custom|PerformTargeting) (Variables|Default|GetTargetingTickInterval) true))
  (Variables|Default|SetTargetingTimerHandle _returnvalue))

(event Custom|PerformTargeting
  (bind (_asynctaskproxy _targetdata) (Ability|Tasks|PerformTargetingAbilityTask "None" "/IcicleLarvaGameplay/Gameplay/Targeting/TP_IcicleLarva_AirCharge.TP_IcicleLarva_AirCharge" false false true)
    (:Targeted
      (if (Utilities|Array|IsNotEmpty (Ability|TargetData|GetAllActorsfromTargetData _targetdata))
        (for _array_element (Ability|TargetData|GetAllActorsfromTargetData _targetdata)
          ; Check if it's hitting an invalid target, and then move to do the punch
          (bind _canbedestroyed (CallFunction|CanTargetbeDirectlyDestroyed _array_element))
          (if _canbedestroyed
            ; Check to see if the targeted actor is a BuildingActor and if so, destroy it if it can be damaged, or trigger the punch when hitting an indestructible one
            (bind _asbuilding_actor (Utilities|Casting|CastToBuildingActor _array_element)
              (:then
                (if (.CanBeDamaged (Class|BPFLAICommonFunctions|GetCanbeDamaged _asbuilding_actor))
                  (Gameplay|Die :self _asbuilding_actor :Killer (Pawn|GetController (CallFunction|GetAvatarasFortPlayerPawn)) :DamageCauser (CallFunction|GetAvatarasFortPlayerPawn) :!class /Script/FortniteGame.BuildingActor)
                  ; Check if the Build Destruction GCN should be played and if yes, play it and reset the timer
                  (if (Variables|Default|GetbShouldPlayDestructionGCN)
                    (Ability|ExecuteGameplayCueOnOwner (Variables|Default|GetBuildDestructionGCNTag))
                    (Variables|Default|SetbShouldPlayDestructionGCN false)
                    (Utilities|Time|SetTimerbyEvent (delegate Custom|ResetDestructionGCN) (ScalableFloat|GetValueatLevel (Variables|Default|GetRow_BuildDestructionDelay))))
                  (else
                    _)))
              (:CastFailed
                _))
            ; If it's something else, set the bool to true to trigger the punch afterwards
            (if (== (Variables|Default|GetbHasHitPawnVehicle) false)
              (Variables|Default|SetbHasHitPawnVehicle true))
            (else
              _)))
        ; If the target has hit a pawn or vehicle, try to do the Punch GA and end the ability
        (if (Variables|Default|GetbHasHitPawnVehicle)
          (Abilities|TryActivateAbilitybyClass (Ability|GetAbilitySystemComponentfromActorInfo) (Variables|Default|GetPunchAbilityClass) true)
          (Ability|EndAbility))))))

(event Ability|EventOnEndAbility
  (params (bWasCancelled :type Boolean))
  (Utilities|Time|ClearandInvalidateTimerbyHandle (Variables|Default|GetTargetingTimerHandle)))

(event Custom|ResetDestructionGCN
  ; Resets the bool enabling the GCN to play again
  (Variables|Default|SetbShouldPlayDestructionGCN true))

(fn CanTargetBeDirectlyDestroyed
  (params (Object :type Object%20Reference))
  (if (Math|Boolean|ORBoolean (Utilities|IsA Object "/Script/FortniteGame.BuildingGameplayActor") (Utilities|IsA Object "/Script/SpireBossRuntime.MajorSpire") (Utilities|IsA Object "/SpireBoss/Gameplay/MinorSpire/BGA_Spire_Minor.BGA_Spire_Minor_C"))
    (return false)
    (else
      (return true))))
