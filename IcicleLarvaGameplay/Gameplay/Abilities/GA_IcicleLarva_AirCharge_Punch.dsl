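; Icicle Larva gameplay ability for the punch tied to its air charge, sending the end-air-charge event on activation.
; It plays the in-air right-hand punch montage, while the server fires a swing GameplayCue carrying the hit-water flag as a float magnitude.
; A targeting task using the ground punch preset applies the on-hit GameplayEffects to each targeted actor.
; The ability ends when the montage completes, blends out, or is interrupted or cancelled.
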(event Ability|EventActivateAbility
  (Ability|SendGameplayEvent (Variables|Default|GetEndAirChargeTag))
  (Ability|Tasks|PlayMontageAndWaitNotify "None" "/IcicleLarvaGameplay/Assets/Animation/Proto/IcicleLarva_Punch_InAir_RHand_Montage.IcicleLarva_Punch_InAir_RHand_Montage" 1.0 "None" false
    (:then
      ; Only server has valid target data
      (Ability|HasAuthority
        (:True
          (GameplayCue|ExecuteGameplayCueOnActor(Burst) (CallFunction|GetAvatarasFortPlayerPawn) (Variables|Default|GetGCN_PunchSwingTag) (Ability|GameplayCue|MakeGameplayCueParameters 0.0 (Math|Conversions|ToFloat(Boolean) (Variables|Default|GetbHitWater))))
          (bind (_asynctaskproxy _targetdata) (Ability|Tasks|PerformTargetingAbilityTask "None" "/IcicleLarvaGameplay/Gameplay/Targeting/TP_IcicleLarva_GroundPunch.TP_IcicleLarva_GroundPunch" true)
            (:Targeted
              (if (Utilities|Array|IsNotEmpty (Ability|TargetData|GetAllActorsfromTargetData _targetdata))
                (for _array_element (Ability|TargetData|GetAllActorsfromTargetData _targetdata)
                  (Fortnite|ApplyGameplayEffectsToActor _array_element (Variables|Default|GetGEsToApplyOnHit) (Transformation|GetActorLocation (CallFunction|GetAvatarasFortPlayerPawn)) false 0 (Ability|GetAbilitySystemComponentfromActorInfo) (Class|AthenaFunctionLibrary|MakeSimpleFortEffectContext (CallFunction|GetAvatarasFortPlayerPawn) (CallFunction|GetAvatarasFortPlayerPawn))))))))))
    (:OnCompleted
      _)
    (:OnBlendOut
      _)
    (:OnInterrupted
      _)
    (:OnCancelled
      _))
  (Ability|EndAbility))
