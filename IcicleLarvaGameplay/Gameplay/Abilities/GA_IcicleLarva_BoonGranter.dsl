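; Gameplay ability that rewards players who damage the Icicle Larva with a boon.
; On activation, the authority reads a damage threshold and binds to the avatar pawn's damage event.
; It totals non-self damage per instigator in a map, using a set to grant once per player at the threshold.
; The boon spawns as a pickup targeted at that player's pawn, and damage and grants report to analytics.
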(event Ability|EventActivateAbility
  (Ability|HasAuthority
    (:True
      (Variables|Default|SetMinDamageToRecieveReward (ScalableFloat|GetValueatLevel (Variables|GameData|GetMinDamageToRecieveReward_GameData)))
      (EventDispatchers|BindEventtoOnDamaged (CallFunction|GetAvatarasFortPlayerPawn) (delegate Custom|OnIcicleLarvaDamaged)))))

(event Custom|OnIcicleLarvaDamaged
  (params (DamagedActor :type Actor%20Object%20Reference) (Damage :type Float%20%28single-precision%29) (InstigatedBy :type Controller%20Object%20Reference) (DamageCauser :type Actor%20Object%20Reference) (HitLocation :type Vector) (FHitComponent :type Primitive%20Component%20Object%20Reference) (BoneName :type Name) (Momentum :type Vector))
  (bind _damageinstigatedby (Variables|Default|GetDamageInstigatedBy))
  (bind _damagemap (Variables|Default|GetDamageMap))
  (bind (_value _returnvalue) (Utilities|Map|Find _damagemap _damageinstigatedby))
  (Variables|Default|SetDamage Damage)
  (bind _output_get (Variables|Default|SetDamageInstigatedBy InstigatedBy))
  ; Ignore self damage
  (if (!= _output_get (Pawn|GetController (CallFunction|GetAvatarasFortPlayerPawn)))
    (CallFunction|AnalyticsRecieveDamage (Variables|Default|GetDamage))
    ; Store total damage caused by the player
    (if _returnvalue
      (Variables|Default|SetDamage (+ (Variables|Default|GetDamage) _value)))
    (Utilities|Map|Add (Variables|Default|GetDamageMap) (Variables|Default|GetDamageInstigatedBy) (Variables|Default|GetDamage))
    ; Should Grant Boon to Player?
    (if (Utilities|Set|ContainsItem (Variables|Default|GetGrantedBoonPlayerSet) (Variables|Default|GetDamageInstigatedBy))
      _
      (elif (>= (Variables|Default|GetDamage) (Variables|Default|GetMinDamageToRecieveReward))
        (Utilities|Set|Add (Variables|Default|GetGrantedBoonPlayerSet) (Variables|Default|GetDamageInstigatedBy))
        ; Grant boon to player
        (bind _asfort_player_controller_athena (Utilities|Casting|CastToFortPlayerControllerAthena (Fortnite|GetControllerfromActor (Variables|Default|GetDamageInstigatedBy))))
        (bind _returnvalue_2 (Inventory|SpawnPickupInWorldWithClassAndItemEntry (Loot|CreateItemEntry (Variables|Default|GetBoonToBeGranted)) "/Script/FortniteGame.FortPickupAthena" (+ (Transformation|GetActorLocation (Game|Player|GetPlayerPawn _asfort_player_controller_athena)) (Variables|Default|GetBoon%20Spawn%20Offset)) "0.000000,0.000000,1.000000" 0 true false false "Player" "TossedByPlayer" 0 false))
        (Pickup|BlueprintSetPickupTarget _returnvalue_2 (Game|Player|GetPlayerPawn _asfort_player_controller_athena) true)
        (CallFunction|AnalyticsBoonGranted)))))

(event Custom|Analytics_RecieveDamage
  (params (DamageAmount :type Float%20%28double-precision%29))
  ; Analytics Events
  (bind _returnvalue (IcicleLarva|GetIcicleLarvaAnalyticsComponent 0))
  (Utilities|IsValid _returnvalue
    (:Is%20Valid
      (Class|IcicleLarvaAnalyticsComponent|NotifyReceiveDamage _returnvalue DamageAmount))))

(event Custom|Analytics_BoonGranted
  ; Analytics Events
  (bind _returnvalue (IcicleLarva|GetIcicleLarvaAnalyticsComponent 0))
  (Utilities|IsValid _returnvalue
    (:Is%20Valid
      (Class|IcicleLarvaAnalyticsComponent|NotifyBoonGranted _returnvalue))))
