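; Gameplay ability that ends a player's IcicleLarva transformation when health runs out.
; Authority intercepts death, plays the outro cue and montage, blocks movement and optionally grants damage immunity, while clients mirror the montage via an outro tag.
; A timer at a tunable fraction of montage length spawns a rift portal and ends the ability.
; Ending removes outro effects and the larva item, sets health, cancels the movement extension, re-equips weapons and reports analytics.
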(event Ability|EventActivateAbility
  (Variables|Default|SetOwnerPawn (CallFunction|GetAvatarasFortPlayerPawn))
  (Ability|HasAuthority
    (:True
      (FortPlayerPawn|BindEventtoOnPreventDeathonOutOfHealth (Variables|Default|GetOwnerPawn) (delegate Custom|OnOutOfHealth)))
    (:False
      (bind (_asynctaskproxy _tagcount) (Ability|Tasks|WaitGameplayTagCountChange (Variables|Default|GetOutroTag))
        (:TagCountChanged
          (if (> _tagcount 0)
            (CallFunction|PlayAnimationClient)
            (GameplayTasks|EndTask _asynctaskproxy)))))))

(event Custom|OnOutOfHealth
  (params (DamageAmount :type Float%20%28single-precision%29) (SourceTags :type Gameplay%20Tag%20Container%20Structure%20%28by%20ref%29) (DamageInstigator :type Controller%20Object%20Reference) (DamageCauser :type Actor%20Object%20Reference))
  (CallFunction|DoOutroSequence))

(event Custom|CancelIcicleLarva
  (bind _returnvalue (CallFunction|GetAvatarasFortPlayerPawn))
  ; Cheat Ke* CancelIcicleLarva
  (Inventory|RemoveItemFromPlayer :PlayerController (Fortnite|GetControllerfromActor _returnvalue) :ItemDefinition (Variables|Default|GetIcicleLarvaItemDefinition) :AmountToRemove -1 :bForceRemoval true :bSkipVerbRemoval false)
  (FortPawn|SetHealth _returnvalue 1e+02)
  (CallFunction|CancelMME)
  (Class|FortPawn|EquipBestWeapon (CallFunction|GetAvatarasFortPlayerPawn)))

(event Custom|DoOutroSequence
  (bind _ownerpawn (Variables|Default|GetOwnerPawn))
  (bind _returnvalue (Ability|GetAbilitySystemComponent _ownerpawn))
  ; Outro Sequence
  (FortPlayerPawn|UnbindallEventsfromOnPreventDeathonOutOfHealth (Variables|Default|GetOwnerPawn))
  (Ability|ExecuteGameplayCueOnOwner (Variables|Default|GetOutroGCNTag) (Class|AthenaFunctionLibrary|MakeSimpleFortEffectContext 0))
  (bind _returnvalue_2 (GameplayEffects|ApplyGameplayEffectToSelf (Ability|GetAbilitySystemComponent (Variables|Default|GetOwnerPawn)) (Variables|Default|GetPreventMoveGEClass) 0.0 (Class|AthenaFunctionLibrary|MakeSimpleFortEffectContext (Variables|Default|GetOwnerPawn))))
  (Variables|Default|SetPreventMoveGEHandle _returnvalue_2)
  (bind _returnvalue_3 (Animation|Utilities|PlayAnimMontage (Variables|Default|GetOwnerPawn) (Variables|Default|GetOutroAnimMontage)))
  (Utilities|Time|SetTimerbyEvent (delegate Custom|HandleOutroEnd) (* _returnvalue_3 (ScalableFloat|GetValueatLevel (Variables|Default|GetRow_MontagePercentToActivateRift))))
  (if (ScalableFloat|GetValueasBool (Variables|Default|GetRow_ShouldApplyDamageImmunity))
    (bind _returnvalue_4 (Ability|GameplayEffect|AssignTagSetbyCallerMagnitude (GameplayEffects|MakeOutgoingSpec _returnvalue (Variables|Default|GetDamageImmunityGEClass) 0.0 (Class|AthenaFunctionLibrary|MakeSimpleFortEffectContext _ownerpawn)) "(TagName=\"SetByCaller.IcicleLarva.SequenceDuration\")" _returnvalue_3))
    (bind _returnvalue_5 (GameplayEffects|ApplyGameplayEffectSpecToSelf _returnvalue _returnvalue_4))
    (Variables|Default|SetDamageImmunityGEHandle _returnvalue_5)))

(event Custom|PlayAnimationClient
  ; Outro Sequence
  (Animation|Utilities|PlayAnimMontage (Variables|Default|GetOwnerPawn) (Variables|Default|GetOutroAnimMontage)))

(event Custom|HandleOutroEnd
  (bind (_returnvalue_location _returnvalue_rotation) (Math|Transform|BreakTransform (Transformation|GetWorldTransform (Variables|/Script/Engine.Character|GetMesh (CallFunction|GetAvatarasFortPlayerPawn)))))
  ; Outro Sequence
  (Building|SpawnBuildingGameplayActor "/IcicleLarvaGameplay/Gameplay/BGA_IcicleLarva_RiftPortal_Item_Athena.BGA_IcicleLarva_RiftPortal_Item_Athena_C" (Math|Transform|MakeTransform _returnvalue_location _returnvalue_rotation) (CallFunction|GetAvatarasFortPlayerPawn) false true)
  (Ability|EndAbility))

(event Custom|CancelMME
  (bind _asfort_movement_comp_character (Utilities|Casting|CastToFortMovementComp_Character (Variables|/Script/Engine.Character|GetCharacterMovement (CallFunction|GetAvatarasFortPlayerPawn))))
  (Class|FortMovementCompCharacter|CancelActiveExtensionwithTag _asfort_movement_comp_character (Variables|Default|GetMovementModeTag)))

(event Ability|EventOnEndAbility
  (params (bWasCancelled :type Boolean))
  (|CallParentFunction bWasCancelled)
  (CallFunction|RemoveOutroGEs)
  (CallFunction|CancelIcicleLarva)
  (bind _asfort_player_controller_athena (Utilities|Casting|CastToFortPlayerControllerAthena (Pawn|GetController (CallFunction|GetAvatarasFortPlayerPawn))))
  (CallFunction|AnalyticsTransformBackPlayer _asfort_player_controller_athena))

(event Custom|RemoveOutroGEs
  (bind _ownerpawn (Variables|Default|GetOwnerPawn))
  (bind _returnvalue (Ability|GetAbilitySystemComponent _ownerpawn))
  (GameplayEffects|RemoveActiveGameplayEffect _returnvalue (Variables|Default|GetPreventMoveGEHandle) -1)
  (GameplayEffects|RemoveActiveGameplayEffect _returnvalue (Variables|Default|GetDamageImmunityGEHandle) -1))

(event Custom|Analytics_TransformBackPlayer
  (params (PlayerController :type Fort%20Player%20Controller%20Object%20Reference))
  (bind _returnvalue (IcicleLarva|GetIcicleLarvaAnalyticsComponent 0))
  (Utilities|IsValid _returnvalue
    (:Is%20Valid
      (Class|IcicleLarvaAnalyticsComponent|NotifyPlayerTransformBack _returnvalue PlayerController))))
