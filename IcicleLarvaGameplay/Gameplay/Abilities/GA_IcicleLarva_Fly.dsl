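; Gameplay ability that launches a player into Icicle Larva flight via a movement mode extension, blocked while that mode is active.
; Falling players fly immediately with zero ascend time; others get a prep camera modifier and takeoff effect during a maximum-hold timer.
; Overridable ProcessInput starts flight on release, with held time clamped into ascend time.
; TryToFly can enforce a minimum hold; ending clears timers, camera and effect.
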(event Ability|EventActivateAbility
  (bind _outfortplayerpawn (CallFunction|GetAvatarasValidFortPlayerPawn)
    (:ValidFortPlayerPawn
      (bind _output_get (Variables|Default|SetPlayerPawn _outfortplayerpawn))
      (if (GameplayTags|HasMatchingGameplayTag _output_get "(TagName=\"MovementMode.Falling\")")
        (CallFunction|StartFlying true)
        (else
          (CallFunction|StartTrytoFlyTimer)
          (CallFunction|ProcessInput))))
    (:AvatarCastFailed
      (Fortnite|FortShippingLog self "Invalid avatar pawn! Cannot initiate Icicle Larva flight." true)
      (Ability|EndAbility))))

(event Custom|TryToFly
  (Utilities|Time|ClearandInvalidateTimerbyHandle (Variables|Default|GetMaxHoldTimerHandle))
  ; Check if time held is at least min time, if not we need to set a timer for the difference
  (if (>= (Variables|Default|GetTimeHeldInput) (ScalableFloat|GetValueatLevel (Variables|Hotfix|GetMinInputHoldTime)))
    (CallFunction|StartFlying false)
    (else
      (bind _returnvalue (Utilities|Time|SetTimerbyEvent (EventDispatchers|CreateEvent :!event Custom|HeldTimeComplete) (- (ScalableFloat|GetValueatLevel (Variables|Hotfix|GetMinInputHoldTime)) (Variables|Default|GetTimeHeldInput))))
      (Variables|Default|SetMinHoldTimerHandle _returnvalue))))

(event Custom|HeldTimeComplete
  (CallFunction|StartFlying false))

(event Custom|StartFlying
  (params (bFromFalling :type Boolean))
  (Class|FortMovementModeExtLogicFlyingIcicleLarva|TryActivateIcicleLarvaMME (Variables|Default|GetPlayerPawn) (Variables|Default|GetIcicleLarvaMMEClass) (Utilities|Struct|MakeFortMovementModeIcicleLarvaCreationData (Math|Float|SelectFloat 0.0 (Math|Float|Clamp(Float) (Variables|Default|GetTimeHeldInput) (ScalableFloat|GetValueatLevel (Variables|Hotfix|GetMinAscendTime)) (ScalableFloat|GetValueatLevel (Variables|Hotfix|GetMaxAscendTime))) bFromFalling)))
  (Ability|EndAbility))

(event Custom|ProcessInput
  ; Derived GAs are expected to override this to fit their input needs - they MUST call StartFlying otherwise the max timer will initiate flight
  (bind (_asynctaskproxy _timeheld) (Ability|Tasks|WaitInputRelease false)
    (:OnRelease
      (Variables|Default|SetTimeHeldInput _timeheld)
      (CallFunction|StartFlying false))))

(event Custom|StartTryToFlyTimer
  (bind _maxinputholdtime (Variables|Hotfix|GetMaxInputHoldTime))
  (bind _returnvalue (ScalableFloat|GetValueatLevel _maxinputholdtime))
  ; Setting time held to max here as ProcessInput will override it if input is released sooner
  (Variables|Default|SetTimeHeldInput _returnvalue)
  (bind _returnvalue_2 (Utilities|Time|SetTimerbyEvent (EventDispatchers|CreateEvent :!event Custom|HeldTimeComplete) _returnvalue))
  (Variables|Default|SetMaxHoldTimerHandle _returnvalue_2)
  (bind _outmodifier (Class|WeaponFunctionLibrary|FindorAddCameraModifierforPawn (Variables|Default|GetPrepAscendCameraModifierClass) (CallFunction|GetAvatarasFortPlayerPawn)))
  (bind _output_get (Variables|Default|SetPrepAscendCameraModifier _outmodifier))
  (CameraModifier|EnableModifier _output_get)
  (bind _returnvalue_3 (Ability|ApplyGameplayEffectToOwner (Variables|Default|GetTakeoffPassiveGEClass)))
  (Variables|Default|SetTakeoffPassiveGEHandle _returnvalue_3))

(event Ability|EventOnEndAbility
  (params (bWasCancelled :type Boolean))
  (CameraModifier|DisableModifier (Variables|Default|GetPrepAscendCameraModifier) false)
  (Utilities|Time|ClearandInvalidateTimerbyHandle (Variables|Default|GetMinHoldTimerHandle))
  (Utilities|Time|ClearandInvalidateTimerbyHandle (Variables|Default|GetMaxHoldTimerHandle))
  (Ability|RemoveGameplayEffectFromOwnerWithHandle (Variables|Default|GetTakeoffPassiveGEHandle))
  (|CallParentFunction bWasCancelled))

(fn CanActivateAbility
  (params (ActorInfo :type Gameplay%20Ability%20Actor%20Info%20Structure) (Handle :type Gameplay%20Ability%20Spec%20Handle%20Structure))
  (bind (_owneractor _avataractor) (Utilities|Struct|BreakGameplayAbilityActorInfo ActorInfo))
  (return 0 (not (GameplayTags|HasMatchingGameplayTag (Utilities|Casting|ToGameplayTagAssetInterface _avataractor) "(TagName=\"MovementMode.IcicleLarva\")"))))

(fn OverrideFailedReason
  (params (FailedReason :type Gameplay%20Tag%20Container%20Structure%20%28by%20ref%29))
  ; Override the failed reason if the MME is active (as this isn't relevant in that case)
  (if (GameplayTags|HasMatchingGameplayTag (CallFunction|GetAvatarasFortPlayerPawn) "(TagName=\"MovementMode.IcicleLarva\")")
    (return 0)
    (else
      (return FailedReason))))
