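; Gameplay ability for the Icicle Larva's special fly move, triggered by holding and releasing jump.
; On input it stores the server world time as the activation time, then waits for the Event.JumpReleased gameplay event.
; When that event arrives, it computes the held duration from server time, passes it through a float Max, and calls TrytoFly.
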(event EventProcessInput
  (Variables|Default|SetAbilityActivatedTime (Fortnite|Sync|Time|GetServerWorldTimeSeconds))
  (Ability|Tasks|WaitGameplayEvent "(TagName=\"Event.JumpReleased\")"
    (:EventReceived
      (Variables|Default|SetTimeHeldInput (Math|Float|Max(Float) (- (Fortnite|Sync|Time|GetServerWorldTimeSeconds) (Variables|Default|GetAbilityActivatedTime))))
      (CallFunction|TrytoFly))))
