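; Icicle Larva ground punch ability: a montage-driven dash and punch with its own passive effect, camera mode and camera modifier.
; The swing notify triggers a non-authority guided-motion dash, toward a targeted actor if allowed and found, otherwise by constant force.
; The hit notify runs targeting, applying hit effects and cues.
; Ending removes the passive and camera modifier; failure reasons are suppressed during Icicle Larva movement mode.
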(event Ability|EventActivateAbility
  (bind _returnvalue (Ability|ApplyGameplayEffectToOwner "/IcicleLarvaGameplay/Gameplay/Effects/GE_IcicleLarva_GroundPunch_Passive.GE_IcicleLarva_GroundPunch_Passive_C"))
  (Variables|Default|SetGroundPunchPassiveGEHandle _returnvalue)
  (bind _outmodifier (Class|WeaponFunctionLibrary|FindorAddCameraModifierforPawn "/IcicleLarvaGameplay/Gameplay/CameraModifiers/B_IcicleLarva_Punch_CameraModifier.B_IcicleLarva_Punch_CameraModifier_C" (CallFunction|GetAvatarasFortPlayerPawn)))
  (bind _output_get (Variables|Default|SetPunchCameraModifier _outmodifier))
  (CameraModifier|EnableModifier _output_get)
  (Camera|PushCameraMode "/IcicleLarvaGameplay/Gameplay/CameraModes/B_IcicleLarva_CameraMode_GroundPunch.B_IcicleLarva_CameraMode_GroundPunch_C")
  (bind (_asynctaskproxy _notifytag) (Ability|Tasks|PlayMontageAndWaitNotify "None" (Variables|Default|GetPunchMontage) 1.0 "None" false)
    (:OnCompleted
      _)
    (:OnBlendOut
      _)
    (:OnInterrupted
      _)
    (:OnCancelled
      _)
    (:OnNotifyBegin
      (Ability|CommitAbility)
      (if (== _notifytag (Variables|Default|GetPunchSwingTag))
        (GameplayCue|ExecuteGameplayCueOnActor(Burst) (CallFunction|GetAvatarasFortPlayerPawn) (Variables|Default|GetGCN_PunchDashTag) (Ability|GameplayCue|MakeGameplayCueParameters 0.0))
        (Ability|HasAuthority
          (:False
            ; Check if it should to Dash Targeting
            (if (ScalableFloat|GetValueasBool (Variables|Default|GetRow_AllowDashTargeting))
              ; Targeting for the Dash to Punch
              (bind (_asynctaskproxy_2 _targetdata) (Ability|Tasks|PerformTargetingAbilityTask "None" "/IcicleLarvaGameplay/Gameplay/Targeting/TP_IcicleLarva_GroundPunch_Dash.TP_IcicleLarva_GroundPunch_Dash")
                (:Targeted
                  (if (Utilities|Array|IsNotEmpty (Ability|TargetData|GetAllActorsfromTargetData _targetdata))
                    ;; If a target was found, use the MoveToActor Dash
                    (Game|ApplyGuidedMotionMME (CallFunction|GetAvatarasFortPlayerPawn) (Variables|Default|GetDashToActorMME) (Utilities|Array|Get(acopy) (Ability|TargetData|GetAllActorsfromTargetData _targetdata)))
                    (else
                      _))))
              (else
                _))
            ; If no target found, do a ConstantForce dash
            (Game|ApplyGuidedMotionMME (CallFunction|GetAvatarasFortPlayerPawn) (Variables|Default|GetDashConstantForceMME))))
        (elif (== _notifytag (Variables|Default|GetPunchHitTag))
          (GameplayCue|ExecuteGameplayCueOnActor(Burst) (CallFunction|GetAvatarasFortPlayerPawn) (Variables|Default|GetGCN_PunchSwingTag) (Ability|GameplayCue|MakeGameplayCueParameters 0.0))
          (CameraModifier|DisableModifier (Variables|Default|GetPunchCameraModifier) false)
          ; Targeting for the punch damage & knockback
          (bind (_asynctaskproxy_3 _targetdata_2) (Ability|Tasks|PerformTargetingAbilityTask "None" "/IcicleLarvaGameplay/Gameplay/Targeting/TP_IcicleLarva_GroundPunch.TP_IcicleLarva_GroundPunch" true)
            (:Targeted
              (if (Utilities|Array|IsNotEmpty (Ability|TargetData|GetAllActorsfromTargetData _targetdata_2))
                (GameplayCue|ExecuteGameplayCueOnActor(Burst) (CallFunction|GetAvatarasFortPlayerPawn) (Variables|Default|GetGCN_PunchHitTag) (Ability|GameplayCue|MakeGameplayCueParameters 0.0))
                (for _array_element (Ability|TargetData|GetAllActorsfromTargetData _targetdata_2)
                  (Fortnite|ApplyGameplayEffectsToActor _array_element (Variables|Default|GetGEsToApplyOnHit) (Transformation|GetActorLocation (CallFunction|GetAvatarasFortPlayerPawn)) false 0 (Ability|GetAbilitySystemComponentfromActorInfo) (Class|AthenaFunctionLibrary|MakeSimpleFortEffectContext (CallFunction|GetAvatarasFortPlayerPawn) (CallFunction|GetAvatarasFortPlayerPawn)))))))))))
  (Ability|EndAbility))

(event Ability|EventOnEndAbility
  (params (bWasCancelled :type Boolean))
  (bind _punchcameramodifier (Variables|Default|GetPunchCameraModifier))
  (Ability|RemoveGameplayEffectFromOwnerWithHandle (Variables|Default|GetGroundPunchPassiveGEHandle))
  (Utilities|IsValid _punchcameramodifier
    (:Is%20Valid
      (CameraModifier|DisableModifier _punchcameramodifier false))))

(fn OverrideFailedReason
  (params (FailedReason :type Gameplay%20Tag%20Container%20Structure%20%28by%20ref%29))
  ; Override the failed reason if the MME is active (as this isn't relevant in that case)
  (if (GameplayTags|HasMatchingGameplayTag (CallFunction|GetAvatarasFortPlayerPawn) "(TagName=\"MovementMode.IcicleLarva\")")
    (return 0)
    (else
      (return FailedReason))))
