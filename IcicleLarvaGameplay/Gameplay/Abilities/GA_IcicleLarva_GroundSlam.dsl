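; Icicle Larva ground slam ability that plays the slam montage, fires an impact cue and applies hit effects to targeted actors.
; The console-variable-gated cue is placed by a trace from low on the capsule along movement direction, falling back to actor location.
; On authority, separate targeting presets for general and building actors feed ApplySlamEffectsToActors.
; That function applies gameplay effects and kills tagged, damageable low-max-health buildings outright, silently for the weakest.
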(event Ability|EventActivateAbility
  (Ability|Tasks|PlayMontageAndWaitNotify "None" "/IcicleLarvaGameplay/Assets/Animation/Proto/IcicleLarva_GroundSlam_Montage_Proto.IcicleLarva_GroundSlam_Montage_Proto" 1.0 "Hit"
    (:then
      (Variables|Default|SetPlayerPawn (CallFunction|GetAvatarasFortPlayerPawn))
      ;[1 0 0 0.5] Can disable GroundSlam GC if it causes crashes again
      (bind _returnvalue (Development|GetConsoleVariableBoolValue "IcicleLarva.GroundSlamCueEnabled"))
      (if _returnvalue
        (bind (_outhit _returnvalue_2) (Collision|LineTraceByChannel (Utilities|Operators|Subtract (Transformation|GetActorLocation (Variables|Default|GetPlayerPawn)) (Math|Vector|MakeVector 0.0 0.0 (* (Components|Capsule|GetScaledCapsuleHalfHeight (Variables|/Script/Engine.Character|GetCapsuleComponent (Variables|Default|GetPlayerPawn))) 0.7))) (+ (Utilities|Operators|Subtract (Transformation|GetActorLocation (Variables|Default|GetPlayerPawn)) (Math|Vector|MakeVector 0.0 0.0 (* (Components|Capsule|GetScaledCapsuleHalfHeight (Variables|/Script/Engine.Character|GetCapsuleComponent (Variables|Default|GetPlayerPawn))) 0.7))) (* (Math|Vector|Normalize (Pawn|Components|CharacterMovement|GetLastUpdateVelocity (Variables|/Script/Engine.Character|GetCharacterMovement (Variables|Default|GetPlayerPawn)))) (Variables|Default|GetGCTraceLength))) "TraceTypeQuery10" false (Utilities|Array|MakeArray (Variables|Default|GetPlayerPawn))))
        (Ability|ExecuteGameplayCueWithParamsOnOwner (Variables|Default|GetGCTag) (Ability|GameplayCue|MakeGameplayCueParameters :NormalizedMagnitude 0.0 :RawMagnitude 0.0 :EffectContext (Class|AthenaFunctionLibrary|MakeSimpleFortEffectContext (Variables|Default|GetPlayerPawn) (Variables|Default|GetPlayerPawn)) :Location (select _returnvalue_2 (.ImpactPoint (Collision|BreakHitResult _outhit)) (Transformation|GetActorLocation (Variables|Default|GetPlayerPawn))) :Normal (select _returnvalue_2 (.ImpactNormal (Collision|BreakHitResult _outhit)) "0.000000,0.000000,1.000000") :GameplayEffectLevel 1 :AbilityLevel 1 :bReplicateLocationWhenUsingMinimalRepProxy false))
        (else
          _))
      (Ability|HasAuthority
        (:True
          (bind (_asynctaskproxy _targetdata) (Ability|Tasks|PerformTargetingAbilityTask "None" "/IcicleLarvaGameplay/Gameplay/Targeting/TP_IcicleLarva_GroundSlam.TP_IcicleLarva_GroundSlam" false false true)
            (:then
              (bind (_asynctaskproxy_2 _targetdata_2) (Ability|Tasks|PerformTargetingAbilityTask "None" "/IcicleLarvaGameplay/Gameplay/Targeting/TP_IcicleLarva_GroundSlam_BuildingActors.TP_IcicleLarva_GroundSlam_BuildingActors" false false true)
                (:Targeted
                  (CallFunction|ApplySlamEffectstoActors (Ability|TargetData|GetAllActorsfromTargetData _targetdata_2)))) ; Building Actors
            )
            (:Targeted
              (CallFunction|ApplySlamEffectstoActors (Ability|TargetData|GetAllActorsfromTargetData _targetdata)))) ; Non-Building Actors
        )))
    (:OnCompleted
      _)
    (:OnBlendOut
      _)
    (:OnInterrupted
      _)
    (:OnCancelled
      _))
  (Ability|EndAbility))

(fn ApplySlamEffectsToActors
  (params (Array :type Array%20of%20Actor%20Object%20References))
  (bind _playerpawn (Variables|Default|GetPlayerPawn))
  (bind _instigatorcontroller (LocalVariables|GetInstigatorController))
  (LocalVariables|SetInstigatorController (Pawn|GetController (Variables|Default|GetPlayerPawn)))
  (for _array_element Array
    (LocalVariables|SetHitActor _array_element)
    (bind _hitactor (LocalVariables|GetHitActor)
      (:then
        (bind _asbuilding_actor (Utilities|Casting|CastToBuildingActor _hitactor)
          (:then
            (if (GameplayTags|HasAnyMatchingGameplayTags _asbuilding_actor (Variables|Default|GetBuildingActorDirectDieTags))
              (if (and (Variables|/Script/Engine.Actor|GetbCanBeDamaged _asbuilding_actor) (not (Variables|/Script/FortniteGame.BuildingActor|GetbIsInvulnerable _asbuilding_actor)))
                (if (< (Health|GetMaxHealth _asbuilding_actor :!class /Script/FortniteGame.BuildingActor) (ScalableFloat|GetValueatLevel (Variables|Default|GetHF_BuildingActorDirectDieHealth)))
                  ; Direct die
                  (if (>= (Health|GetMaxHealth _asbuilding_actor :!class /Script/FortniteGame.BuildingActor) (ScalableFloat|GetValueatLevel (Variables|Default|GetHF_BuildingActorSilentDieHealth)))
                    (Gameplay|Die :self _asbuilding_actor :Killer _instigatorcontroller :DamageCauser _playerpawn :!class /Script/FortniteGame.BuildingActor) ; Die
                    (else
                      (Gameplay|Die _asbuilding_actor "(TagName=\"Gameplay.Effect.InstantDeath.Silent\")" _instigatorcontroller _playerpawn :!class /Script/FortniteGame.BuildingActor) ; Silent Die
                    ))
                  (else
                    _)))
              (else
                _))
            ; Apply full game effects
            (Fortnite|ApplyGameplayEffectsToActor (LocalVariables|GetHitActor) (Variables|Default|GetGEsToApplyOnHit) (Transformation|GetActorLocation (CallFunction|GetAvatarasFortPlayerPawn)) false 0 (Ability|GetAbilitySystemComponentfromActorInfo) (Class|AthenaFunctionLibrary|MakeSimpleFortEffectContext (CallFunction|GetAvatarasFortPlayerPawn) (CallFunction|GetAvatarasFortPlayerPawn))))
          (:CastFailed
            _)))
      (:else))))
