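; Gameplay ability for the Icicle Larva's heat vision beam, which charges up, fires for a set duration, then ends with a cooldown.
; Activation starts cosmetic targeting, a passive effect and build-up cue, then timers drive the Icicle Larva pawn component's beam and end the ability.
; An unblockable effect prevents cancellation until a minimum time passes or an outro tag change cancels early.
; On end, clients remove the looping beam cue locally rather than waiting for the server.
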(event Ability|EventActivateAbility
  (CallFunction|StartUnblockablePeriod)
  (bind _outfortplayerpawn (CallFunction|GetAvatarasValidFortPlayerPawn)
    (:ValidFortPlayerPawn
      (bind _output_get (Variables|Default|SetOwnerPawn _outfortplayerpawn))
      (bind _asfort_pawn_component_icicle_larva (Utilities|Casting|CastToFortPawnComponent_IcicleLarva (Actor|GetComponentbyClass _output_get (Variables|Default|GetPawnComponentClass)))
        (:then
          (Variables|Default|SetIcicleLarvaPawnComponent _asfort_pawn_component_icicle_larva)
          ; Start cosmetic targeting for UI/FX
          ; Not calling the corresponding stop function as StopFiringBeam will handle that on end ability
          (Beam|StartCosmeticBeamTargeting (Variables|Default|GetIcicleLarvaPawnComponent))
          (Utilities|Time|SetTimerbyEvent (EventDispatchers|CreateEvent :!event Custom|StartBeam) (Beam|GetBeamChargeUpTime (Variables|Default|GetIcicleLarvaPawnComponent)))
          ; Setting up Beam Cancel Timer
          (Utilities|Time|SetTimerbyEvent (EventDispatchers|CreateEvent :!event Custom|AllowBeamCancel) (ScalableFloat|GetValueatLevel (Variables|Default|GetRow_MinimalDurationBeforeCancel)))
          ; Passive GE is needed for things like granted tags on sim proxies
          (bind _returnvalue (Ability|ApplyGameplayEffectToOwner (Variables|Default|GetPassiveGEClass)))
          (Variables|Default|SetPassiveGEHandle _returnvalue)
          (Ability|AddGameplayCueToOwner (Variables|Default|GetGCTag_Beam_BuildUp) (Class|AthenaFunctionLibrary|MakeSimpleFortEffectContext (Variables|Default|GetOwnerPawn) (Variables|Default|GetOwnerPawn)))
          ; Checking if player has started the outro sequence and if so end the unblockable period and cancel the ability. This is necessary to override the period of time where the ability can't be canceled
          (bind _asynctaskproxy (Ability|Tasks|WaitGameplayTagCountChange "(TagName=\"Granted.IcicleLarva.Outro\")")
            (:then
              (Variables|Default|SetOutroAsyncTask _asynctaskproxy))
            (:TagCountChanged
              (CallFunction|EndUnblockablePeriod)
              (Ability|CancelAbility))))
        (:CastFailed
          (Fortnite|FortShippingLog self "Could not find valid Icicle Larva pawn component! Ending ability." true)
          (Ability|EndAbility))))
    (:AvatarCastFailed
      (Fortnite|FortShippingLog self "No valid FortPlayerPawn Avatar! Ending ability." true)
      (Ability|EndAbility))))

(event Custom|AllowBeamCancel
  ; Setting up Beam Cancel Timer
  (CallFunction|EndUnblockablePeriod))

(event Custom|StartBeam
  (Ability|RemoveGameplayCueFromOwner (Variables|Default|GetGCTag_Beam_BuildUp))
  ; Important to call StartFiringBeam before the Gameplay Cue runs any logic so it can use correct values of beam start/end points
  (Beam|StartFiringBeam (Variables|Default|GetIcicleLarvaPawnComponent))
  (Ability|AddGameplayCueToOwner (Variables|Default|GetGCTag_Beam) (Class|AthenaFunctionLibrary|MakeSimpleFortEffectContext (Variables|Default|GetOwnerPawn) (Variables|Default|GetOwnerPawn)))
  ; Set timer to end the beam
  (bind _returnvalue (Utilities|Time|SetTimerbyEvent (EventDispatchers|CreateEvent :!event K2_EndAbility) (Beam|GetBeamDuration (Variables|Default|GetIcicleLarvaPawnComponent))))
  (Variables|Default|SetEndBeamTimerHandle _returnvalue))

(event Ability|EventOnEndAbility
  (params (bWasCancelled :type Boolean))
  (|CallParentFunction bWasCancelled)
  (GameplayTasks|EndTask (Variables|Default|GetOutroAsyncTask))
  (Ability|CommitAbilityCooldown false)
  (CallFunction|EndUnblockablePeriod)
  (Ability|RemoveGameplayEffectFromOwnerWithHandle (Variables|Default|GetPassiveGEHandle))
  (CallFunction|RemoveBeamGC)
  (Utilities|IsValid (Variables|Default|GetIcicleLarvaPawnComponent)
    (:Is%20Valid
      (Beam|StopFiringBeam (Variables|Default|GetIcicleLarvaPawnComponent)))
    (:Is%20Not%20Valid
      (Fortnite|FortShippingLog self "Cached Icicle Larva pawn component is invalid! Cannot call StopFiringBeam." true))))

(event Custom|RemoveBeamGC
  (Utilities|IsServer
    (:true
      (Ability|RemoveGameplayCueFromOwner (Variables|Default|GetGCTag_Beam)))
    (:false
      ; Client shouldn't have to wait on server to remove the gameplay cue (which it would have to otherwise because it's not authoritative and does not have a valid prediction key after the latent nodes)
      (Projectile|RemoveGameplayCue%28Looping%29Local (CallFunction|GetAvatarasFortPlayerPawn) (Variables|Default|GetGCTag_Beam) (Ability|GameplayCue|MakeGameplayCueParameters 0.0)))))

(fn EndUnblockablePeriod
  (Ability|SetCanbeCanceled true)
  (Ability|RemoveGameplayEffectFromOwnerWithHandle (Variables|Default|GetUnblockableGEHandle)))

(fn StartUnblockablePeriod
  (Ability|SetCanbeCanceled false)
  (bind _returnvalue (Ability|ApplyGameplayEffectToOwner "/IcicleLarvaGameplay/Gameplay/Effects/GE_IcicleLarva_HeatVision_Unblockable.GE_IcicleLarva_HeatVision_Unblockable_C"))
  (Variables|Default|SetUnblockableGEHandle _returnvalue))
