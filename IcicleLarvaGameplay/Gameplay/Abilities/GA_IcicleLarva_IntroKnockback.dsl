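; Gameplay ability for the Icicle Larva intro that knocks back targeted actors around the avatar pawn.
; On event activation it fires a knockback gameplay cue on the owner, then runs a targeting task with the intro pushback preset.
; Each returned actor receives the configured gameplay effects, using the pawn's location and effect context, and the ability then ends.
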(event Ability|EventActivateAbilityFromEvent
  (params (EventData :type Gameplay%20Event%20Data%20Structure%20%28by%20ref%29))
  (Ability|ExecuteGameplayCueOnOwner (Variables|Default|GetKnockbackGCNTag) (Class|AthenaFunctionLibrary|MakeSimpleFortEffectContext (CallFunction|GetAvatarasFortPlayerPawn)))
  (bind (_asynctaskproxy _targetdata) (Ability|Tasks|PerformTargetingAbilityTask "None" "/IcicleLarvaGameplay/Gameplay/Targeting/TP_IcicleLarva_Intro_Pushback.TP_IcicleLarva_Intro_Pushback" true false true)
    (:Targeted
      (for _array_element (Ability|TargetData|GetAllActorsfromTargetData _targetdata)
        (Fortnite|ApplyGameplayEffectsToActor _array_element (Variables|Default|GetGEsToApply) (Transformation|GetActorLocation (CallFunction|GetAvatarasFortPlayerPawn)) false 0 (Ability|GetAbilitySystemComponentfromActorInfo) (Class|AthenaFunctionLibrary|MakeSimpleFortEffectContext (CallFunction|GetAvatarasFortPlayerPawn) (CallFunction|GetAvatarasFortPlayerPawn))))
      (Ability|EndAbility))))
