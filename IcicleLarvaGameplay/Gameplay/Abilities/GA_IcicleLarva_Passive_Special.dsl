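; Passive gameplay ability that equips the IcicleLarva special player with supporting components while active.
; On activation, the authority adds the configurable pawn component and caches it as IcicleLarva_Component_Special.
; Client and server each add a non-replicated pickup-interact override component to the controller.
; On end, the authority calls ResetPlayer before destroying the pawn component; both sides destroy the override, relying on bAllowAnyoneToDestroyMe.
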(event Ability|EventActivateAbility
  (bind _outfortplayerpawn (CallFunction|GetAvatarasValidFortPlayerPawn)
    (:ValidFortPlayerPawn
      (Utilities|FlowControl|Sequence
        (:then_0
          (Ability|HasAuthority
            (:True
              (bind _returnvalue (Components|AddDynamicComponentbyClass _outfortplayerpawn (Variables|Default|GetPawnComponentClass)))
              (bind _asicicle_larva_component_special (Utilities|Casting|CastToIcicleLarva_Component_Special _returnvalue))
              (Variables|Default|SetIcicleLarvaComponent _asicicle_larva_component_special))))
        (:then_1
          ; Not a replicated component, added on both client and server
          (bind _returnvalue_2 (Game|AddComponentbyClass (Pawn|GetController _outfortplayerpawn) "/IcicleLarvaGameplay/Gameplay/IcicleLarva_Component_PickupInteractOverride.IcicleLarva_Component_PickupInteractOverride_C"))
          (Variables|Default|SetPickupInteractOverrideComponent _returnvalue_2))))))

(event Ability|EventOnEndAbility
  (params (bWasCancelled :type Boolean))
  (bind _iciclelarvacomponent (Variables|Default|GetIcicleLarvaComponent))
  (|CallParentFunction bWasCancelled)
  (bind _outfortplayerpawn (CallFunction|GetAvatarasValidFortPlayerPawn)
    (:ValidFortPlayerPawn
      (Utilities|FlowControl|Sequence
        (:then_0
          (Ability|HasAuthority
            (:True
              (Class|IcicleLarvaComponentSpecial|ResetPlayer _iciclelarvacomponent)
              (Components|DestroyDynamicComponent _outfortplayerpawn _iciclelarvacomponent))))
        (:then_1
          ; Can be destroyed in this scope because the component has the bAllowAnyoneToDestroyMe flag
          (Components|DestroyComponent (Variables|Default|GetPickupInteractOverrideComponent)))))))
