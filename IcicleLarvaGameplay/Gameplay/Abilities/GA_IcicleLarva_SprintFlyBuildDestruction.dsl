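; Gameplay ability for Icicle Larva sprint-flying that destroys buildings the player hits and charges energy for each one.
; While active, a looping timer runs a targeting preset task on the avatar pawn.
; Damageable building actors are killed, except BuildingGameplayActors and major or minor Spires, and energy is consumed on server and client.
; A bool reset by a delay timer throttles the destruction gameplay cue, and ending the ability clears the targeting timer.
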(event Ability|EventActivateAbility
  (|CallParentFunction)
  (bind _returnvalue (Utilities|Time|SetTimerbyEvent (delegate Custom|PerformTargeting) (Variables|Default|GetTargetingTickInterval) true))
  (Variables|Default|SetTargetingTimerHandle _returnvalue)
  (Variables|Default|SetOwnerPawn (CallFunction|GetAvatarasFortPlayerPawn)))

(event Custom|PerformTargeting
  (bind (_asynctaskproxy _targetdata) (Ability|Tasks|PerformTargetingAbilityTask "None" "/IcicleLarvaGameplay/Gameplay/Targeting/TP_IcicleLarva_SprintFlyBuildDestruction.TP_IcicleLarva_SprintFlyBuildDestruction" true false true)
    (:Targeted
      (if (Utilities|Array|IsNotEmpty (Ability|TargetData|GetAllActorsfromTargetData _targetdata))
        (for _array_element (Ability|TargetData|GetAllActorsfromTargetData _targetdata)
          ; Check if it's hitting any invalid targets
          (bind _canbedestroyed (CallFunction|CanTargetbeDirectlyDestroyed _array_element))
          (if _canbedestroyed
            ; Check to see if the targeted actor is a BuildingActor and if so, destroy it if it can be damaged
            (bind _asbuilding_actor (Utilities|Casting|CastToBuildingActor _array_element))
            (if (.CanBeDamaged (Class|BPFLAICommonFunctions|GetCanbeDamaged _asbuilding_actor))
              (Gameplay|Die :self _asbuilding_actor :Killer (Pawn|GetController (Variables|Default|GetOwnerPawn)) :DamageCauser (Variables|Default|GetOwnerPawn) :!class /Script/FortniteGame.BuildingActor)
              ; Consuming Energy on the Server and Client after destroying a building actor
              (CallFunction|ConsumeEnergyonDestruction)
              (CallFunction|ConsumeEnergyClient)
              ; Check if the Build Destruction GCN should be played and if yes, play it and reset the timer
              (if (Variables|Default|GetbShouldPlayDestructionGCN)
                (Ability|ExecuteGameplayCueOnOwner (Variables|Default|GetBuildDestructionGCNTag))
                (Variables|Default|SetbShouldPlayDestructionGCN false)
                (Utilities|Time|SetTimerbyEvent (delegate Custom|ResetDestructionGCN) (ScalableFloat|GetValueatLevel (Variables|Default|GetRow_BuildDestructionDelay)))))))))))

(event Custom|ConsumeEnergyClient
  ; Consuming Energy on the Server and Client after destroying a building actor
  (CallFunction|ConsumeEnergyonDestruction))

(event Custom|ConsumeEnergyOnDestruction
  ; Consume Energy on Build Destruction
  (Energy|UseEnergy (Actor|GetComponentbyClass (Variables|Default|GetOwnerPawn) "/Script/EnergyRuntime.FortComponent_Energy") (ScalableFloat|GetValueatLevel (Variables|Default|GetRow_SprintFlyEnergyConsumptionOnDestruction)) true 0))

(event Ability|EventOnEndAbility
  (params (bWasCancelled :type Boolean))
  (Utilities|Time|ClearandInvalidateTimerbyHandle (Variables|Default|GetTargetingTimerHandle)))

(event Custom|ResetDestructionGCN
  ; Resets the bool enabling the GCN to play again
  (Variables|Default|SetbShouldPlayDestructionGCN true))

(fn CanTargetBeDirectlyDestroyed
  (params (Object :type Object%20Reference))
  (if (Math|Boolean|ORBoolean (Utilities|IsA Object "/Script/FortniteGame.BuildingGameplayActor") (Utilities|IsA Object "/Script/SpireBossRuntime.MajorSpire") (Utilities|IsA Object "/SpireBoss/Gameplay/MinorSpire/BGA_Spire_Minor.BGA_Spire_Minor_C"))
    (return false)
    (else
      (return true))))
