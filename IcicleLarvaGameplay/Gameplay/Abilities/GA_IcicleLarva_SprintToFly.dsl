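; GA_IcicleLarva_SprintToFly is a gameplay ability for the Icicle Larva flight feature, switching the player pawn into a flying movement mode, likely from sprinting as the name suggests.
; On activation it checks that the avatar is a valid FortPlayerPawn.
; It then calls TryActivateIcicleLarvaMME with the configured movement mode extension class and inline creation data, and ends the ability right afterward on that same valid-pawn path.
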(event Ability|EventActivateAbility
  (bind _outfortplayerpawn (CallFunction|GetAvatarasValidFortPlayerPawn)
    (:ValidFortPlayerPawn
      (Class|FortMovementModeExtLogicFlyingIcicleLarva|TryActivateIcicleLarvaMME _outfortplayerpawn (Variables|Default|GetMMEClass) (Utilities|Struct|MakeFortMovementModeIcicleLarvaCreationData 0.0 true))
      (Ability|EndAbility))))
