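; Gameplay ability for IcicleLarva that temporarily overrides the storm sickness kill threshold of the avatar's player.
; On activation it finds the TrackStormDamageComponent on the pawn's player state, caches its original threshold and applies the modified value.
; On end it restores the cached threshold and calls Reset on the component.
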(event Ability|EventActivateAbility
  (bind _returnvalue (Fortnite|Utilities|AttemptGetPlayerStateforPawn (CallFunction|GetAvatarasFortPlayerPawn)))
  (Utilities|IsValid _returnvalue
    (:Is%20Valid
      (bind _output_get (Variables|Default|SetTrackStormDamage_Component (Actor|GetComponentbyClass _returnvalue "/StormSickness/TrackStormDamageComponent.TrackStormDamageComponent_C")))
      (Utilities|IsValid _output_get
        (:Is%20Valid
          (Variables|Default|SetOriginalStormSicknessKillThreshold (Variables|/StormSickness/TrackStormDamageComponent.TrackStormDamageComponent_C|GetfStormSicknessKillThreshold _output_get))
          (Variables|/StormSickness/TrackStormDamageComponent.TrackStormDamageComponent_C|SetfStormSicknessKillThreshold (Variables|Default|GetModifiedStormSicknessKillThreshold) (Variables|Default|GetTrackStormDamage_Component)))))))

(event Ability|EventOnEndAbility
  (params (bWasCancelled :type Boolean))
  (Utilities|IsValid (Variables|Default|GetTrackStormDamage_Component)
    (:Is%20Valid
      (Variables|/StormSickness/TrackStormDamageComponent.TrackStormDamageComponent_C|SetfStormSicknessKillThreshold (Variables|Default|GetOriginalStormSicknessKillThreshold) (Variables|Default|GetTrackStormDamage_Component))
      (Class|TrackStormDamageComponent|Reset (Variables|Default|GetTrackStormDamage_Component) true))))
